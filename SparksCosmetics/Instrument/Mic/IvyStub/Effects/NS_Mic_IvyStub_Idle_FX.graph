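# Idle Niagara effect for the IvyStub microphone cosmetic, combining spiral ribbons with mesh glows.
# NE_SpiralRibbon bursts partitioned particles around a sampled skeletal mesh location, using sine/cosine offsets and normalized-position curves for twisting, pulsing, tapered ribbons with two renderers.
# Scaled owner velocity bends the ribbons mid-length as the mic moves.
# Pickaxe_Glows places 16 indexed meshes by bone on a second skeletal mesh, setting a visibility tag from the front-end preview flag.
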NiagaraSystem /SparksCosmetics/Instrument/Mic/IvyStub/Effects/NS_Mic_IvyStub_Idle_FX
dependencies {
 usedDataInterfaces [&NiagaraDataInterfaceCurve, &NiagaraDataInterfaceColorCurve, &NiagaraDataInterfaceSkeletalMesh]
 usedDynamicInputs [&MakeFloatFromVector]
 usedModules [
  &self:SystemSpawnScript.SystemScriptSource.SystemScriptGraph.NiagaraNodeAssignment_0.SetVariables_#1
  &self:SystemSpawnScript.SystemScriptSource.SystemScriptGraph.NiagaraNodeAssignment_1.SetVariables_#2
  &self:SystemSpawnScript.SystemScriptSource.SystemScriptGraph.NiagaraNodeAssignment_2.SetVariables_#3
  &Timeline
  &self:SystemSpawnScript.SystemScriptSource.SystemScriptGraph.NiagaraNodeAssignment_3.SetVariables_#4
  &SystemState
 ]
}
emitters {
 NE_SpiralRibbon {
  emitterSpawnScript [/SetVariables(Emitter.ParticlePerRibbon:NiagaraInt32=$System.ParticlePerRibbon, Emitter.RibbonNumber:NiagaraInt32=$System.RibbonNumber) {isSetParameters=true}]
  emitterUpdateScript [
   &EmitterState(
    Inactive Response=::NewEnumerator1
    Life Cycle Mode=::NewEnumerator1
    Loop Behavior=::NewEnumerator1
    Loop Duration Mode=::NewEnumerator1
    UseLoopDelay=false
   )
   &SpawnBurst_Instantaneous(
    Spawn Count=&Multiply_Int(A=$Emitter.ParticlePerRibbon, B=$Emitter.RibbonNumber)
    Use Spawn Probability=false
    UseLoopCountLimit=false
   )
  ]
  particleSpawnScript [
   &InitializeParticle(
    Color=<rgba 1 1 1 0.7>
    Lifetime=5
    Ribbon Width Mode=::NewEnumerator1
    Sprite Size Mode=::NewEnumerator3
    UsePositionOffset=false
    Write Material Random=true
   )
   &PartitionParticles(
    ID to Partition:NiagaraInt32=&ReturnExecIndex()
    Number Of Partitions=$Emitter.RibbonNumber
    Particles Per Partition:NiagaraInt32=$Emitter.ParticlePerRibbon
    Partition Mode:ENiagara_PartitionMode_v2=::NewEnumerator1
   )
   /SetVariables(Particles.NormalizedRibbonPos:NiagaraFloat=&OneMinusFloat(Float=$Particles.PartitionParticles.Normalized Index In Partition)) {isSetParameters=true}
   /SetVariables(
    Particles.SpiralFreq=0.8
    Particles.SpiralStrength:NiagaraFloat=&Multiply_Float(
     A=15
     B=&FloatFromCurve(
      FloatCurve=DataInterface() {propertyValues {Curve {keys [
          {arriveTangent=10.2716, interpMode=RCIM_Cubic, leaveTangent=10.2716, tangentMode=RCTM_User, time=0, value=-0.025}
          {interpMode=RCIM_Cubic, tangentMode=RCTM_User, time=0.2, value=1}
          {arriveTangent=-0.504, interpMode=RCIM_Cubic, leaveTangent=-0.504, tangentMode=RCTM_User, time=0.6, value=0.95}
          {arriveTangent=-0.1348, interpMode=RCIM_Cubic, leaveTangent=-0.1348, tangentMode=RCTM_User, time=1, value=0}
         ]}}}
      CurveIndex=$Particles.NormalizedRibbonPos
     )
    )
   ) {isSetParameters=true}
   /SetVariables(Particles.RibbonU0Override=&OneMinusFloat(Float=$Particles.NormalizedRibbonPos)) {isSetParameters=true}
   /SetVariables() {isSetParameters=true}
  ]
  particleUpdateScript [
   &ParticleState(Kill Particles When Lifetime Has Elapsed=false)
   &SkeletalMeshLocation(
    Bone / Socket Index=3
    MaskSpawnGroup=false
    Mesh Sampling Type=::NewEnumerator5
    Position Sampling=::NewEnumerator2
    Skeletal Mesh=$System.Skeletal Mesh
    Socket Sampling Mode=::NewEnumerator5
    Use Sampled Mesh Attribute as Ribbon ID=false
   )
   /SetVariables(Particles.GoalPos:NiagaraPosition=&AddVectorToPosition(
     Position=$Particles.SkeletalMeshLocation.SampledPosition
     Vector=&Lerp_Vector(A=<0 0 -10>, B=<0 0 30>, Alpha=$Particles.NormalizedRibbonPos)
    )) {isSetParameters=true}
   /SetVariables(Particles.SpiralStrength:NiagaraFloat=&Multiply_Float(
     A=&Multiply_Float(
      A=2
      B=&Waveform(
       Global Phase=&Add_Float(A=$Emitter.Age, B=$Particles.NormalizedRibbonPos)
       "Waveform [1]"=::NewEnumerator12
       "[1] Frequency"=0.5
       "[1] Amplitude Min/Max"=<1 1.25>
      )
     )
     B=&FloatFromCurve(
      FloatCurve=DataInterface() {propertyValues {Curve {keys [
          {arriveTangent=3.1146, interpMode=RCIM_Cubic, leaveTangent=3.1146, leaveTangentWeight=6.4519, tangentMode=RCTM_User, time=0, value=0.2}
          {arriveTangent=1.7735, interpMode=RCIM_Cubic, leaveTangent=1.7735, tangentMode=RCTM_User, time=0.5, value=1.25}
          {arriveTangentWeight=0.1022, interpMode=RCIM_Cubic, leaveTangentWeight=0.0683, tangentMode=RCTM_User, time=0.7984, value=1.3124}
          {arriveTangent=-7.7827, interpMode=RCIM_Cubic, leaveTangent=-7.7827, tangentMode=RCTM_User, time=0.9979, value=0.2324}
         ]}}}
      CurveIndex=$Particles.NormalizedRibbonPos
     )
    )) {isSetParameters=true}
   /SetVariables(Particles.SpiralTime:NiagaraFloat=&Add_Float(
     A=&Add_Float(A=&Multiply_Float(A=$System.Age, B=0.1), B=$Particles.NormalizedRibbonPos)
     B=&Multiply_FloatByInt(
      Float=&Divide_Float(A=$Particles.SpiralFreq, B=&MakeFloatFromInt(Int=$Emitter.RibbonNumber))
      Integer=$Particles.PartitionParticles.Partition
     )
    )) {isSetParameters=true}
   /SetVariables(Particles.SpiralOffset:Vector3f=&MakeVector(
     X=&Sine(
      Normalized Angle=$Particles.SpiralTime
      Period=$Particles.SpiralFreq
      Scale=&Multiply_Float(A=$Particles.SpiralStrength, B=1.5)
      Bias=0
     )
     Y=&Cosine(
      Normalized Angle=$Particles.SpiralTime
      Period=$Particles.SpiralFreq
      Scale=&Multiply_Float(A=$Particles.SpiralStrength, B=1.5)
      Bias=0
     )
     Z=0
    )) {isSetParameters=true}
   &Color(Color=&MultiplyLinearColor(
     Linear Color=&ColorFromCurve(
      ColorCurve=DataInterface() {propertyValues {
        AlphaCurve {keys [<at 0.0365 1 RCIM_Cubic>,<at 0.7167 1>,<at 1 0>]}
        BlueCurve {keys [<at 0.0057 1>,<at 0.7608 1.5>]}
        GreenCurve {keys [<at 0.0057 2>,<at 0.7608 1.5>]}
        RedCurve {keys [<at 0.0057 0.5>,<at 0.7608 0.375>]}
       }}
      CurveIndex=$Particles.NormalizedRibbonPos
     )
     Scale Factor=<1 1 1 1>
    )) {module=Color001}
   /SetVariables(Particles.Position=&AddVectorToPosition(Position=$Particles.GoalPos, Vector=$Particles.SpiralOffset)) {isSetParameters=true}
   &InheritSourceMovement(
    Applied Position Falloff Distance=275
    Applied Position Scale=&VectorFromFloat(Value=&Multiply_Float(
      A=-0.5
      B=&FloatFromCurve(
       FloatCurve=DataInterface() {propertyValues {Curve {keys [
           {arriveTangent=0.9742, interpMode=RCIM_Cubic, leaveTangent=0.9742, tangentMode=RCTM_User, time=0, value=0}
           <at 0.5 0.4 RCIM_Cubic>
           {arriveTangent=-0.7756, interpMode=RCIM_Cubic, leaveTangent=-0.7756, tangentMode=RCTM_User, time=1, value=0}
          ]}}}
       CurveIndex=$Particles.NormalizedRibbonPos
      )
     ))
    ApplyForce=false
    ApplyPosition=true
    ApplyPositionFalloff=true
    ApplyPositionFalloffCurve=false
    Source Velocity:Vector3f=&TransformVector(Vector=$Engine.Owner.Velocity, Source Space=::World, Destination Space=::Local)
   )
   &SolveForcesAndVelocity(Clamp Velocity=false, Limit Acceleration=false)
   /SetVariables(Particles.RibbonFacing=&NormalizeVector(
     Vector To Normalize=&TransformVector(Vector=$Particles.SpiralOffset, Destination Space=::Simulation)
     Safe Normalize=true
     Safe Fallback Threshold=0
     Safe Fallback Vector=<0 0 0.001>
    )) {isSetParameters=true}
   &ScaleRibbonWidth(Ribbon Width Scale=&Multiply_Float(
     A=&FloatFromCurve(
      FloatCurve=DataInterface() {propertyValues {Curve {keys [
          {arriveTangent=0.5186, interpMode=RCIM_Cubic, leaveTangent=0.5186, leaveTangentWeight=3.2425, tangentMode=RCTM_User, time=0, value=0}
          {arriveTangentWeight=0.2872, interpMode=RCIM_Cubic, leaveTangentWeight=0.1567, tangentMode=RCTM_User, time=0.6, value=1}
          {arriveTangentWeight=0.2515, interpMode=RCIM_Cubic, tangentMode=RCTM_User, time=1, value=0}
         ]}}}
      CurveIndex=$Particles.NormalizedRibbonPos
     )
     B=&Waveform(
      Global Phase=&Add_Float(A=$Emitter.Age, B=$Particles.NormalizedRibbonPos)
      "Waveform [1]"=::NewEnumerator12
      "[1] Frequency"=0.5
      "[1] Amplitude Min/Max"=<0.9 1.25>
     )
    ))
   /SetVariables(Particles.GlowWidth:NiagaraFloat=&Multiply_Float(A=$Particles.RibbonWidth, B=1.3)) {isSetParameters=true}
   /SetVariables(Particles.GlowColor=&MultiplyLinearColor(
     Linear Color=&MultiplyLinearColor(Linear Color=$Particles.Color, Scale Factor=&MakeVector4_FromLinearColor(LinearColor=<rgba 1 1 1 1>))
     Scale Factor=&MakeVector4(
      X=1
      Y=1
      Z=1
      W=&FloatFromCurve(
       FloatCurve=DataInterface() {propertyValues {Curve {keys [
           {arriveTangent=8.2127, interpMode=RCIM_Cubic, leaveTangent=8.2127, tangentMode=RCTM_User, time=0, value=0}
           {arriveTangent=0.4596, arriveTangentWeight=0.232, leaveTangentWeight=2.2538, time=0.2, value=1}
           {arriveTangent=0.0486, interpMode=RCIM_Cubic, leaveTangent=0.0486, tangentMode=RCTM_User, time=0.75, value=1}
           {arriveTangent=-4.4982, interpMode=RCIM_Cubic, leaveTangent=-4.4982, tangentMode=RCTM_User, time=1, value=0}
          ]}}}
       CurveIndex=$Particles.NormalizedRibbonPos
       Scale Curve=&OneMinusFloat(Float=$System.Timeline.NormalizedTime)
      )
     )
    )) {isSetParameters=true}
   &DynamicMaterialParameters(
    Index 0 Param 1=1
    Index 1 Param 2=1
    Index 1 Param 3=&FloatFromCurve(
     FloatCurve=DataInterface() {propertyValues {Curve {keys [
         {arriveTangent=-2.6187, interpMode=RCIM_Cubic, leaveTangent=-2.6187, tangentMode=RCTM_User, time=0, value=1.5}
         {arriveTangent=-0.275, interpMode=RCIM_Cubic, leaveTangent=-0.275, time=0.7, value=0}
         {interpMode=RCIM_Cubic, tangentMode=RCTM_User, time=1, value=1.225}
        ]}}}
     CurveIndex=$Particles.PartitionParticles.Normalized Index In Partition
     Scale Curve=0.5
    )
    Module.Index0Param0Write=true
    Module.Index0Param1Write=true
    Module.Index0Param2Write=true
    Module.Index0Param3Write=true
    Module.Index1Param0Write=true
    Module.Index1Param1Write=true
    Module.Index1Param2Write=true
    Module.Index1Param3Write=true
    Write Parameter Index 1=true
   )
   &ScaleColor(
    Curve Index=$Particles.NormalizedRibbonPos
    Linear Color Curve=DataInterface() {propertyValues {
      AlphaCurve {keys [<at 0 0>, {arriveTangent=1.4043, interpMode=RCIM_Cubic, time=0.1636, value=1}, <at 0.7167 1>, <at 1 0>]}
      BlueCurve {keys [<at 0.0057 1>,<at 0.7608 1.5>]}
      GreenCurve {keys [<at 0.0057 1>,<at 0.7608 1.5>]}
      RedCurve {keys [<at 0.0057 0.5029>,<at 0.7608 0.375>]}
     }}
    Scale Mode=::NewEnumerator2
   ) {module=ScaleColor001}
   &ScaleColor(
    Scale RGB=&MakeVectorFromLinearColorRGB(Color=&MultiplyLinearColor(
      Linear Color=$StackContext.Color
      Scale Factor=&MakeVector4_FromFloat(Float=&FNTimeOfDayScalar(Night Scale=1, Morning Scale=1, Evening Scale=1.05))
     ))
    ScaleA=true
    ScaleRGB=true
   ) {module=ScaleColor002}
  ]
  rendererClasses [&NiagaraRibbonRendererProperties]
  renderers [
   {rendererClass=&NiagaraRibbonRendererProperties}
   {rendererClass=&NiagaraRibbonRendererProperties, rendererIndex=1}
  ]
 }
 Pickaxe_Glows {
  emitterUpdateScript [
   &EmitterState(
    Inactive Response=::NewEnumerator1
    Life Cycle Mode=::NewEnumerator1
    Loop Behavior=::NewEnumerator1
    Loop Duration Mode=::NewEnumerator1
    UseLoopDelay=false
   )
   &SpawnBurst_Instantaneous(Spawn Count=16, Use Spawn Probability=false, UseLoopCountLimit=true)
  ]
  particleSpawnScript [
   &InitializeParticle(
    Color Mode=::NewEnumerator0
    Lifetime=0.25
    Mesh Scale Mode=::NewEnumerator3
    Sprite Size Mode=::NewEnumerator3
    Uniform Sprite Size=5
    UsePositionOffset=false
    Write Material Random=true
   )
   &PartitionParticles(Number Of Partitions=16)
   /SetVariables(Particles.VisibilityTag=&MakeCustomIntFromBool(Bool=$User.bIsFrontEndPreview, True INT=1, False INT=0)) {isSetParameters=true}
  ]
  particleUpdateScript [
   &ParticleState(Kill Particles When Lifetime Has Elapsed=false)
   &ScaleMeshSize(Scale Factor=&VectorFromFloat(Value=$System.MeshScaleRemap))
   &SkeletalMeshLocation(Bone / Socket Index=$Engine.ExecIndex, Bone Sampling Mode=::NewEnumerator4, MaskSpawnGroup=false, Skeletal Mesh=$System.Skeletal Mesh2)
   /SetVariables(Particles.MeshIndex=&ReturnExecIndex()) {isSetParameters=true}
  ]
  rendererClasses [&NiagaraMeshRendererProperties]
  renderers [
   {rendererClass=&NiagaraMeshRendererProperties, rendererIndex=0}
   {rendererClass=&NiagaraMeshRendererProperties, rendererIndex=1}
  ]
 }
}
summary {
 systemName=NS_Mic_IvyStub_Idle_FX
 emitters [
  {emitterName=NE_SpiralRibbon, rendererClasses [&NiagaraRibbonRendererProperties]}
  {emitterName=Pickaxe_Glows, rendererClasses [&NiagaraMeshRendererProperties]}
 ]
 userVariables [User.bIsFrontEndPreview {type=NiagaraBool}]
}
