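# Animation Blueprint for the IvyStub microphone cosmetic that continuously spins its mouthpiece and adds springy secondary motion.
# Each update derives a steady yaw from game time and sets the handle offset to zero; sine-based float math is left unconnected.
# The AnimGraph applies both additively to dyn_handle and dyn_mouth_piece over the input pose.
# A gravity-free Anim Dynamics node then gives the mouthpiece damped, tightly limited sway relative to the handle.
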AnimBlueprint /SparksCosmetics/Instrument/Mic/IvyStub/Meshes/IvyStub_Mic_AnimBP
target_skeleton=&/IvyStub_Mic_Skeleton
graphs [
 AnimGraph {nodes [
   n1 = "Transform (Modify) Bone - Bone: dyn_handle"(
    ComponentPose=Local To Component(LocalPose=Input Pose())
    Alpha=1
    Translation=Get StoneFloatA() -> (StoneFloatA:Vector)
    Rotation="0,0,0"
    Scale="1,1,1"
   ) {
    rotationMode=BMM_Additive
    rotationSpace=BCS_BoneSpace
    translationMode=BMM_Additive
    translationSpace=BCS_BoneSpace
    alphaBoolBlend {blendOption=Linear}
    boneToModify {boneName=dyn_handle}
   }
   n2 = "Transform (Modify) Bone - Bone: dyn_mouth_piece"(ComponentPose=@n1, Alpha=1, Translation="0,0,0", Rotation=Get Mic Piece Rotator() -> (Mic Piece Rotator:Rotator), Scale="1,1,1") {
    rotationMode=BMM_Additive
    rotationSpace=BCS_BoneSpace
    translationMode=BMM_Additive
    translationSpace=BCS_BoneSpace
    alphaBoolBlend {blendOption=Linear}
    boneToModify {boneName=dyn_mouth_piece}
   }
   n3 = Anim Dynamics - Bone: dyn_mouth_piece(ComponentPose=@n2, Alpha=0.8, ExternalForce="0,0,0") {
    angularDampingOverride=0.9
    bAngularSpring=true
    bGravityOverrideInSimSpace=true
    bLinearSpring=true
    bOverrideAngularDamping=true
    bOverrideLinearDamping=true
    gravityScale=0
    linearDampingOverride=0.9
    linearSpringConstant=10
    simulationSpace=BoneRelative
    alphaBoolBlend {blendOption=Linear}
    boundBone {boneName=dyn_mouth_piece}
    componentLinearAccScale=<0.1 0.1 0.1>
    componentLinearVelScale=<2 2 2>
    physicsBodyDefinitions [{
      sphereCollisionRadius=0
      boundBone {boneName=dyn_mouth_piece}
      boxExtents=<30 30 30>
      constraintSetup {angularLimitsMax=<0 0 0>, angularLimitsMin=<0 0 0>, angularTarget=<0 0 0>, linearAxesMax=<0.25 0.25 0.25>, linearAxesMin=<-0.25 -0.25 -0.25>}
      localJointOffset=<0 -3 0>
     }]
    relativeSpaceBone {boneName=handle}
    simSpaceSettings {maxAngularAcceleration=5000, maxAngularVelocity=5000}
   }
   Output Pose(Result=Component To Local(ComponentPose=@n3))
  ]}
 EventGraph {nodes [
   Event BlueprintUpdateAnimation()
   > Set StoneFloatA(StoneFloatA:Vector=MakeVector(Y=0)) -> (Output_Get:Vector)
   n1 = GetGameTimeInSeconds()
   n2 = float / float(A=@n1, B=-500) {is=title K2Node_CallFunction}
   > Set Mic Piece Rotator(Mic Piece Rotator:Rotator=MakeRotator(Yaw=float * float(A=@n2, B=15000))) -> (Output_Get:Rotator)
   TryGetPawnOwner()
   GetGameTimeInSeconds()
   n3 = float * float(A=1, B=1.25)
   float * float(A="Sin (Radians)"(A=float * float(A=@n3, B=@n1)), B=0.2)
   n4 = float * float(A=1, B=1)
   float * float(A="Sin (Radians)"(A=float * float(A=@n4, B=@n1)), B=0.2)
   n5 = float * float(A=1, B=1.25)
   float * float(A="Sin (Radians)"(A=float * float(A=@n5, B=@n1)), B=0.2)
  ]}
]
variables [
 Mic Piece Rotator {type=Rotator}
 StoneFloatA {type=Vector}
]
